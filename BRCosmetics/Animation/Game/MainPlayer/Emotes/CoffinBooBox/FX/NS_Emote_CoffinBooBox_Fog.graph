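# Niagara fog effect for the CoffinBooBox emote, combining several fog, glow and light emitters with some variants disabled; most show no detail here.
# Falling_Fog continuously spawns drifting, sinking greenish sprites, with front-end modifiers scaling velocity and a BlendSpace module shifting world/local position by owner speed.
# Outward_Fog_Puffs fires two timed bursts of short-lived, drag-slowed sprites, tagged by spawn group for renderer visibility.
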NiagaraSystem /BRCosmetics/Animation/Game/MainPlayer/Emotes/CoffinBooBox/FX/NS_Emote_CoffinBooBox_Fog
dependencies {
 usedDataInterfaces [&NiagaraDataInterfaceCurve, &NiagaraDataInterfaceColorCurve, &NiagaraDataInterfaceVectorCurve]
 usedDynamicInputs [&FNTimeOfDayColor, &ColorFromCurve]
 usedModules [
  &SystemState
  &BlendSpaceHelper
  &self:SystemSpawnScript.SystemScriptSource.SystemScriptGraph.NiagaraNodeAssignment_0.SetVariables_#1
  &self:SystemSpawnScript.SystemScriptSource.SystemScriptGraph.NiagaraNodeAssignment_2.SetVariables_#2
 ]
}
emitters {
 Coffin_InnerFog {rendererClasses [], renderers []}
 Coffin_fog {rendererClasses [], renderers []}
 Coffin_fog001 {bEnabled=false, rendererClasses [], renderers []}
 Falling_Fog {
  emitterUpdateScript [
   &EmitterState(Delay First Loop Only=true, Life Cycle Mode=::NewEnumerator1, Loop Delay=1, Loop Duration=1, UseLoopDelay=true)
   &SpawnRate(SpawnRate=18, Use Spawn Probability=false)
  ]
  particleSpawnScript [
   &InitializeParticle(
    Lifetime Max=1.8
    Lifetime Min=1.4
    Lifetime Mode=::NewEnumerator1
    Sprite Rotation Mode=::NewEnumerator2
    Sprite Size=<35 38>
    Sprite Size Mode=::NewEnumerator4
    UsePositionOffset=false
    Write Material Random=true
   )
   &ShapeLocation(
    Handle Radius=1
    Large Radius=12
    MaskSpawnGroup=false
    Non Uniform Scale=<2.15 0.85 1>
    Offset=<10 0 0>
    Offset Mode=::NewEnumerator0
    Override Seed=false
    Rotation Mode=::NewEnumerator2
    Shape Primitive=::NewEnumerator3
    UseOwnerScale=true
    Yaw / Pitch / Roll=<0.25 0 0.25>
   )
   &KillParticlesInVolume(Box Size:Vector3f=<20 50 20>, Kill Shape=::NewEnumerator1, Origin Offset:Vector3f=<8.3119 0 15>)
   &AddVelocity(Velocity=<0 -7 0>, Velocity Speed Scale=&FN_FrontEndFloatModifier(Gameplay=0, Front End=1))
  ]
  particleUpdateScript [
   &ParticleState()
   &SpriteRotationRate(Rotation Rate=&RandomRangeFloat(Minimum=-60, Maximum=60))
   &ScaleSpriteSize(Uniform Curve Sprite Scale=DataInterface() {propertyValues {Curve {keys [{tangentMode=RCTM_User, time=0, value=0.7}, {tangentMode=RCTM_User, time=1, value=1.2}]}}})
   &Drag(Use Linear Drag=true, Use Rotational Drag=true)
   &ScaleVelocity(Velocity Scale=&VectorFromCurve(
     VectorCurve=DataInterface() {propertyValues {XCurve {keys [<at 0 0>,<at 1 0>]}, YCurve {keys [<at 0 0 RCIM_Cubic>,<at 1 1 RCIM_Cubic>]}, ZCurve {keys [<at 0 0>,<at 1 0>]}}}
     Scale Curve=&FN_FrontEndVector3Modifier(Gameplay=<0 0 0>, Front End=<1 1 1>)
    )) {enabled=false}
   &GravityForce(Gravity=<0 0 -15>)
   &SolveForcesAndVelocity(Clamp Velocity=false, Limit Acceleration=false)
   &ScaleColor(
    Linear Color Curve=DataInterface() {propertyValues {
      AlphaCurve {keys [<at 0 0>,<at 0.1012 0.5>,<at 0.907 0>]}
      BlueCurve {keys [<at 0 0.618>]}
      GreenCurve {keys [<at 0 1.5>]}
      RedCurve {keys [<at 0 0.4491>]}
     }}
    Scale Mode=::NewEnumerator2
   )
   &BlendSpace("Position Blend (World <-> Local)":NiagaraFloat=&MakeCustomFloatFromBool(
     Bool=&SetBoolByFloatComparison(A=&VectorLength(Vector=$Engine.Owner.Velocity), B=310)
     True Float=1
     False Float=&FloatFromCurve(FloatCurve=DataInterface() {propertyValues {Curve {keys [<at 0 1 RCIM_Cubic>,<at 0.9978 0.5 RCIM_Cubic>]}}})
     Translation Contribution:NiagaraFloat=1
     Rotation Contribution:NiagaraFloat=1
    )) {module=BlendSpace001}
   &CameraOffset(Camera Offset Amount=10)
  ]
  rendererClasses [&NiagaraSpriteRendererProperties]
  renderers [
   {rendererClass=&NiagaraSpriteRendererProperties}
   {rendererClass=&NiagaraSpriteRendererProperties, rendererIndex=1}
  ]
 }
 Falling_FogLWE {bEnabled=false, rendererClasses [], renderers []}
 Inner_FogSprites {rendererClasses [], renderers []}
 Inner_Glow {rendererClasses [], renderers []}
 Light {rendererClasses [], renderers []}
 Outward_Fog_Puffs {
  emitterUpdateScript [
   &EmitterState(Life Cycle Mode=::NewEnumerator1, Loop Behavior=::NewEnumerator1, Loop Duration=1, UseLoopDelay=false)
   &SpawnBurst_Instantaneous(Spawn Count=25, Spawn Time=0.45, Use Spawn Probability=false, UseLoopCountLimit=false)
   &SpawnBurst_Instantaneous(Spawn Count=25, Spawn Group=1, Spawn Time=0.75, Use Spawn Probability=false, UseLoopCountLimit=false) {module=SpawnBurst_Instantaneous001}
  ]
  particleSpawnScript [
   &InitializeParticle(
    Lifetime Max=0.4
    Lifetime Min=0.38
    Lifetime Mode=::NewEnumerator1
    Position Offset=<10 0 0>
    Sprite Rotation Mode=::NewEnumerator2
    Sprite Size Mode=::NewEnumerator3
    Uniform Sprite Size=50
    UsePositionOffset=false
    Write Material Random=true
   )
   /SetVariables(Particles.VisibilityTag=&SpawnGroup()) {isSetParameters=true}
   &ShapeLocation(
    Handle Radius=1
    Large Radius=8
    MaskSpawnGroup=false
    Non Uniform Scale=<1.8707 0.84 1>
    Offset=<9 0 -2>
    Offset Mode=::NewEnumerator0
    Override Seed=false
    Rotation Mode=::NewEnumerator2
    Shape Primitive=::NewEnumerator3
    UseOwnerScale=true
    Yaw / Pitch / Roll=<0.25 0 0.25>
   )
   &KillParticlesInVolume(Box Size:Vector3f=<50 50 20>, Kill Shape=::NewEnumerator1, Origin Offset:Vector3f=<0 0 11>)
   &AddVelocity(Module.Use Falloff Distance=false, Origin Offset=<0 0 25>, Override Seed=false, Velocity Mode=::NewEnumerator1, Velocity Speed=50)
   &CameraOffset(Camera Offset Amount=15)
  ]
  particleUpdateScript [
   &ParticleState()
   &ScaleSpriteSize(Uniform Curve Sprite Scale=DataInterface() {propertyValues {Curve {keys [{tangentMode=RCTM_User, time=0, value=0.2}, {tangentMode=RCTM_User, time=1, value=1.2}]}}})
   &Drag(Drag=5, Use Linear Drag=true, Use Rotational Drag=true)
   &SolveForcesAndVelocity(Clamp Velocity=false, Limit Acceleration=false)
   &ScaleColor(
    Linear Color Curve=DataInterface() {propertyValues {
      AlphaCurve {keys [<at 0 0>,<at 0.1992 0.3>,<at 1 0>]}
      BlueCurve {keys [<at 0 1.0276>]}
      GreenCurve {keys [<at 0 2>]}
      RedCurve {keys [<at 0 0.6979>]}
     }}
    Scale Mode=::NewEnumerator2
   )
  ]
  rendererClasses [&NiagaraSpriteRendererProperties]
  renderers [
   {rendererClass=&NiagaraSpriteRendererProperties}
   {rendererClass=&NiagaraSpriteRendererProperties, rendererIndex=1}
  ]
 }
 Outward_fog1 {bEnabled=false, rendererClasses [], renderers []}
 Outward_fog2 {bEnabled=false, rendererClasses [], renderers []}
 Upward_Fog {rendererClasses [], renderers []}
}
summary {
 systemName=NS_Emote_CoffinBooBox_Fog
 emitters [
  {emitterName=Coffin_fog, rendererClasses []}
  {bEnabled=false, emitterName=Falling_FogLWE, rendererClasses []}
  {emitterName=Inner_Glow, rendererClasses []}
  {emitterName=Inner_FogSprites, rendererClasses []}
  {emitterName=Light, rendererClasses []}
  {emitterName=Upward_Fog, rendererClasses []}
  {bEnabled=false, emitterName=Outward_fog2, rendererClasses []}
  {bEnabled=false, emitterName=Outward_fog1, rendererClasses []}
  {emitterName=Outward_Fog_Puffs, rendererClasses [&NiagaraSpriteRendererProperties]}
  {emitterName=Falling_Fog, rendererClasses [&NiagaraSpriteRendererProperties]}
  {emitterName=Coffin_InnerFog, rendererClasses []}
  {bEnabled=false, emitterName=Coffin_fog001, rendererClasses []}
 ]
 userVariables [
  User.bIsFrontEnd {type=NiagaraBool}
  User.bIsFrontEndPreview {type=NiagaraBool}
 ]
}
