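# Niagara effect for the prop spawn in the CoffinBooBox main-player emote.
# The Bats emitter bursts a few short-lived mesh particles that scale in and out, scatter on curl noise with curve-driven upward force and drag, and fade from green to near-black.
# LWE emitters for sparkles, airglow, smoke and swirl smoke accompany it, their internals not shown.
# SpookySmokeLWE and BatsLWE are disabled.
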NiagaraSystem /BRCosmetics/Animation/Game/MainPlayer/Emotes/CoffinBooBox/FX/NS_Emote_CoffinBooBox_PropSpawn
dependencies {
 usedDataInterfaces [&NiagaraDataInterfaceVectorCurve, &NiagaraDataInterfaceCurve, &NiagaraDataInterfaceColorCurve]
 usedDynamicInputs [
  &RandomRangeVector
  &FNTimeOfDayColor
  &ColorFromCurve
 ]
 usedModules [
  &SystemState
  &self:SystemSpawnScript.SystemScriptSource.SystemScriptGraph.NiagaraNodeAssignment_0.SetVariables_#1
 ]
}
emitters {
 AirglowLWE {rendererClasses [], renderers []}
 Bats {
  emitterUpdateScript [
   &EmitterState()
   &SpawnBurst_Instantaneous(Spawn Count=10, Use Spawn Probability=false, UseLoopCountLimit=false)
  ]
  particleSpawnScript [
   &InitializeParticle(
    Lifetime Max=0.85
    Lifetime Min=0.6
    Lifetime Mode=::NewEnumerator1
    Mesh Scale Mode=::NewEnumerator1
    Mesh Uniform Scale Max=0.4
    Mesh Uniform Scale Min=0.32
    UsePositionOffset=false
    Write Material Random=true
   )
   &ShapeLocation(Box Size=<20 40 20>, MaskSpawnGroup=false, Override Seed=false, Shape Primitive=::NewEnumerator2, Surface Only=false, UseOwnerScale=true) {enabled=false}
   &AddVelocity(
    Module.Use Falloff Distance=false
    Override Seed=false
    Velocity Mode=::NewEnumerator1
    Velocity Speed=&RandomRangeFloat(Minimum=75, Maximum=120)
   )
  ]
  particleUpdateScript [
   &ParticleState()
   &ScaleMeshSize(Scale Factor=&VectorFromCurve(VectorCurve=DataInterface() {propertyValues {
       XCurve {keys [
         <at 0 0 RCIM_Cubic>
         {arriveTangentWeight=0.1067, interpMode=RCIM_Cubic, leaveTangentWeight=0.3202, tangentMode=RCTM_User, time=0.1873, value=1}
         {arriveTangentWeight=0.3202, interpMode=RCIM_Cubic, leaveTangentWeight=0.1067, tangentMode=RCTM_User, time=0.804, value=1}
         <at 1 0 RCIM_Cubic>
        ]}
       YCurve {keys [
         <at 0 0 RCIM_Cubic>
         {arriveTangentWeight=0.1067, interpMode=RCIM_Cubic, leaveTangentWeight=0.3202, tangentMode=RCTM_User, time=0.1873, value=1}
         {arriveTangentWeight=0.3202, interpMode=RCIM_Cubic, leaveTangentWeight=0.1067, tangentMode=RCTM_User, time=0.804, value=1}
         <at 1 0 RCIM_Cubic>
        ]}
       ZCurve {keys [
         <at 0 0 RCIM_Cubic>
         {arriveTangentWeight=0.1067, interpMode=RCIM_Cubic, leaveTangentWeight=0.3202, tangentMode=RCTM_User, time=0.1814, value=1}
         {arriveTangentWeight=0.3202, interpMode=RCIM_Cubic, leaveTangentWeight=0.1067, tangentMode=RCTM_User, time=0.804, value=1}
         <at 1 0 RCIM_Cubic>
        ]}
      }}))
   &CurlNoiseForce(
    Bias Noise=false
    Noise Frequency=&RandomRangeFloat(Minimum=25, Maximum=75)
    Noise Strength=300
    Pan Noise=false
   )
   &GravityForce(Gravity=&VectorFromCurve(VectorCurve=DataInterface() {propertyValues {XCurve {keys [<at 0 0>,<at 1 0>]}, YCurve {keys [<at 0 0>,<at 1 0>]}, ZCurve {keys [<at 0.2618 0 RCIM_Cubic>,<at 0.4997 300 RCIM_Cubic>]}}}))
   &Drag(
    Drag=&FloatFromCurve(FloatCurve=DataInterface() {propertyValues {Curve {keys [<at 0 0 RCIM_Cubic>,<at 0.6353 1 RCIM_Cubic>]}}})
    Use Linear Drag=true
    Use Rotational Drag=true
   )
   &SolveForcesAndVelocity(Clamp Velocity=false, Limit Acceleration=false)
   &ScaleColor(
    Linear Color Curve=DataInterface() {propertyValues {
      AlphaCurve {keys [<at 0 0>,<at 0.0537 1>,<at 0.4689 1>,<at 1 0>]}
      BlueCurve {keys [<at 0 0.1169>,<at 0.2859 0.0084>]}
      GreenCurve {keys [<at 0 0.555>,<at 0.2859 0.0156>]}
      RedCurve {keys [<at 0 0.0083>,<at 0.2859 0.0037>]}
     }}
    Scale Mode=::NewEnumerator2
   )
  ]
  rendererClasses [&NiagaraMeshRendererProperties]
  renderers [{rendererClass=&NiagaraMeshRendererProperties, rendererIndex=0}]
 }
 BatsLWE {bEnabled=false, rendererClasses [], renderers []}
 SmokeLWE {rendererClasses [], renderers []}
 SparklesLWE {rendererClasses [], renderers []}
 SpookySmokeLWE {bEnabled=false, rendererClasses [], renderers []}
 SwirlSmokeLWE {rendererClasses [], renderers []}
}
summary {
 systemName=NS_Emote_CoffinBooBox_PropSpawn
 emitters [
  {emitterName=SparklesLWE, rendererClasses []}
  {emitterName=AirglowLWE, rendererClasses []}
  {emitterName=SmokeLWE, rendererClasses []}
  {bEnabled=false, emitterName=SpookySmokeLWE, rendererClasses []}
  {bEnabled=false, emitterName=BatsLWE, rendererClasses []}
  {emitterName=Bats, rendererClasses [&NiagaraMeshRendererProperties]}
  {emitterName=SwirlSmokeLWE, rendererClasses []}
 ]
 userVariables [User.NormalizedNotifyProgress {type=NiagaraFloat, defaultValue=0, description=""}]
}
